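; Quest actor for an archive server in the Season 20 Battle Pass Week 9 narrative quest, animating its computer mesh by interaction state.
; Overrides the parent's VFX and audio functions, turning the animation on for pre-interaction and off for post-interaction and no-VFX states.
; A shared toggle event sets the "Active" scalar material parameter on the pre-interaction mesh only on remote instances for the local player controller, keeping the visual per player.
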(event EventBeginPlay
  (|CallParentFunction) ; This node is disabled and will not be called.\nDrag off pins to build functionality.
)

(event Collision|EventActorBeginOverlap
  (params (OtherActor :type Actor%20Object%20Reference))
  (|CallParentFunction OtherActor) ; This node is disabled and will not be called.\nDrag off pins to build functionality.
)

(event EventTick
  (params (DeltaSeconds :type Float%20%28single-precision%29))
  (|CallParentFunction DeltaSeconds) ; This node is disabled and will not be called.\nDrag off pins to build functionality.
)

(event Custom|ToggleComputerAnimation
  (params (PlayerController :type Fort%20Player%20Controller%20Object%20Reference) (ShouldAnimate :type Boolean))
  (Networking|SwitchHasAuthority
    (:Remote
      (if (Pawn|IsLocalPlayerController PlayerController)
        (Rendering|Material|SetScalarParameterValueonMaterials (Variables|Components|GetSM_Quest_PreInteraction) "Active" (select ShouldAnimate 1.0 0.0))))))

(fn NoVFX&Audio
  (params (PlayerController :type Fort%20Player%20Controller%20Object%20Reference))
  (|CallParentFunction PlayerController)
  (CallFunction|ToggleComputerAnimation PlayerController))

(fn PostInteractionVFX&Audio
  (params (PlayerController :type Fort%20Player%20Controller%20Object%20Reference))
  (|CallParentFunction PlayerController)
  (CallFunction|ToggleComputerAnimation PlayerController))

(fn PreInteractionVFX&Audio
  (params (PlayerController :type Fort%20Player%20Controller%20Object%20Reference))
  (|CallParentFunction PlayerController)
  (CallFunction|ToggleComputerAnimation PlayerController true))

(fn ConstructionScript
  (|CallParentFunction))
